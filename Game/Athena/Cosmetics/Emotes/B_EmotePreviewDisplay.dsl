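; Front-end preview Blueprint that plays an emote, emoji or shout on the displayed hero pawn.
; Shouts spawn a tracked sound and push a preview emoji; emojis async-load a preview or fallback montage, play it, then apply emote flags to the pet.
; The special-action event stops the montage, spawns a transition effect and repeats the preview logic.
; New previews and EndPlay clean up spawned sounds; EndPlay also removes pet flags.
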(event Custom|SpecialCaseShoutHandler
  (bind _spawnedcomponent (Audio|SpawnSoundComponent (Variables|Default|GetShoutItemDefinition) 0 (Variables|Default|GetScene) "0, 0, 0"))
  (Utilities|Array|Add (Variables|Default|GetDynamicallySpawnedSounds) _spawnedcomponent)
  (Utilities|FlowControl|Sequence
    (:then_0
      (bind _returnvalue (FrontEndContext|GetHeroPlayerPawnforCurrentDisplayedItem (BlueprintContext|GetFortFrontEndContext)))
      ;[1 1 1] Setup Emoji Preview Component
      (Utilities|IsValid (Actor|GetComponentbyClass _returnvalue "/Script/FortniteGame.FortEmojiPreviewComponent")
        (:Is%20Valid
          _)
        (:Is%20Not%20Valid
          (bind _returnvalue_2 (AddComponent|NODE_AddFortEmojiPreviewComponent-1 :self _returnvalue))))
      (Emoji|PushPreviewEmoji _returnvalue_2 (Variables|Default|GetShoutItemDefinition)))
    (:then_1
      (Default|CallOnEmoteSetupFinished))))

(event Visuals|EventPerformSpecialAction
  (params (ActionName :type Name :default "None"))
  (CallFunction|CleanupDynamicSounds)
  ;[1 1 1] On 'Play Again' we want to stop the montage so FX state is properly cleanup up and reset
  (bind _returnvalue (FrontEndContext|GetHeroPlayerPawnforCurrentDisplayedItem (BlueprintContext|GetFortFrontEndContext)))
  (Utilities|IsValid _returnvalue
    (:Is%20Valid
      (Animation|Montage|MontageStop (Components|SkeletalMesh|GetAnimInstance (Variables|/Script/Engine.Character|GetMesh _returnvalue)) 0.0 0)
      (Effects|SpawnEmitteratLocation "/Game/Effects/Fort_Effects/UI/Transition/P_FrontEndCharacterTransition.P_FrontEndCharacterTransition" (.location (Transformation|GetActorTransform _returnvalue)) (.rotation (Transformation|GetActorTransform _returnvalue))))
    (:Is%20Not%20Valid
      _))
  ;[1 1 1] Get Animation Asset. Use Emoji Preview Animation if available
  (bind _returnvalue_2 (FrontEndContext|GetHeroPlayerPawnforCurrentDisplayedItem (BlueprintContext|GetFortFrontEndContext)))
  (bind _asathena_shout_item_definition (Utilities|Casting|CastToAthenaShoutItemDefinition (stale Emote|EventPreviewEmote _output_get))
    (:then
      (Variables|Default|SetShoutItemDefinition _asathena_shout_item_definition)
      (CallFunction|SpecialCaseShoutHandler))
    (:CastFailed
      (bind _asathena_emoji_item_definition (Utilities|Casting|CastToAthenaEmojiItemDefinition (stale Emote|EventPreviewEmote _output_get))
        (:then
          (bind _output_get (Variables|Default|SetEmojiItemDefinition _asathena_emoji_item_definition))
          (if (Utilities|IsValidSoftObjectReference (Variables|/Script/FortniteGame.AthenaEmojiItemDefinition|GetPreviewAnimation _output_get))
            (Variables|Default|SetAnimation (Variables|/Script/FortniteGame.AthenaEmojiItemDefinition|GetPreviewAnimation _output_get))
            (else
              (bind _output_get_2 (Variables|Default|SetAnimation (Emote|GetFrontendPreviewAnimation (stale Emote|EventPreviewEmote _output_get) (Customization|GetCharacterBodyType _returnvalue_2) 0 (Customization|GetCharacterGender _returnvalue_2))))))
          (bind _object (Utilities|AsyncLoadAsset _output_get_2)
            (:Completed
              (bind _asanim_montage (Utilities|Casting|CastToAnimMontage _object)
                (:then
                  (bind _returnvalue_3 (FrontEndContext|GetHeroPlayerPawnforCurrentDisplayedItem (BlueprintContext|GetFortFrontEndContext)))
                  ;[1 1 1] Setup Emoji Preview Component
                  (Utilities|IsValid (Variables|Default|GetEmojiItemDefinition)
                    (:Is%20Valid
                      (Utilities|IsValid (Actor|GetComponentbyClass _returnvalue_3 "/Script/FortniteGame.FortEmojiPreviewComponent")
                        (:Is%20Valid
                          _)
                        (:Is%20Not%20Valid
                          (bind _returnvalue_4 (AddComponent|NODE_AddFortEmojiPreviewComponent-0 :self _returnvalue_3))))
                      (Emoji|PushPreviewEmoji _returnvalue_4 (Variables|Default|GetEmojiItemDefinition)))
                    (:Is%20Not%20Valid
                      _))
                  (Animation|ResetClothandDynamics _returnvalue_3)
                  (Variables|Default|SetStartSectionName (Emote|PickMontageSection (stale Emote|EventPreviewEmote _output_get) _asanim_montage _returnvalue_3))
                  (Animation|PlayLocalAnimMontage _returnvalue_3 _asanim_montage 1.0 (Variables|Default|GetStartSectionName) true)
                  (Utilities|FlowControl|Sequence
                    (:then_0
                      (Utilities|FlowControl|Delay 0.1)
                      (CallFunction|ApplyEmotetoPet false))
                    (:then_1
                      _)))
                (:CastFailed
                  _))
              (Default|CallOnEmoteSetupFinished))))
        (:CastFailed
          (Variables|Default|SetEmojiItemDefinition 0))))))

(event Emote|EventPreviewEmote
  (params (Emote :type Fort%20Montage%20Item%20Definition%20Base%20Object%20Reference))
  (CallFunction|CleanupDynamicSounds)
  (bind _output_get (Variables|Default|SetDanceOrEmojiItemDefinition Emote))
  ;[1 1 1] Get Animation Asset. Use Emoji Preview Animation if available
  (bind _returnvalue (FrontEndContext|GetHeroPlayerPawnforCurrentDisplayedItem (BlueprintContext|GetFortFrontEndContext)))
  (bind _asathena_shout_item_definition (Utilities|Casting|CastToAthenaShoutItemDefinition _output_get)
    (:then
      (Variables|Default|SetShoutItemDefinition _asathena_shout_item_definition)
      (CallFunction|SpecialCaseShoutHandler))
    (:CastFailed
      (bind _asathena_emoji_item_definition (Utilities|Casting|CastToAthenaEmojiItemDefinition _output_get)
        (:then
          (bind _output_get_2 (Variables|Default|SetEmojiItemDefinition _asathena_emoji_item_definition))
          (if (Utilities|IsValidSoftObjectReference (Variables|/Script/FortniteGame.AthenaEmojiItemDefinition|GetPreviewAnimation _output_get_2))
            (Variables|Default|SetAnimation (Variables|/Script/FortniteGame.AthenaEmojiItemDefinition|GetPreviewAnimation _output_get_2))
            (else
              (bind _output_get_3 (Variables|Default|SetAnimation (Emote|GetFrontendPreviewAnimation _output_get (Customization|GetCharacterBodyType _returnvalue) 0 (Customization|GetCharacterGender _returnvalue))))))
          (bind _object (Utilities|AsyncLoadAsset _output_get_3)
            (:Completed
              (bind _asanim_montage (Utilities|Casting|CastToAnimMontage _object)
                (:then
                  (bind _returnvalue_2 (FrontEndContext|GetHeroPlayerPawnforCurrentDisplayedItem (BlueprintContext|GetFortFrontEndContext)))
                  ;[1 1 1] Setup Emoji Preview Component
                  (Utilities|IsValid (Variables|Default|GetEmojiItemDefinition)
                    (:Is%20Valid
                      (Utilities|IsValid (Actor|GetComponentbyClass _returnvalue_2 "/Script/FortniteGame.FortEmojiPreviewComponent")
                        (:Is%20Valid
                          _)
                        (:Is%20Not%20Valid
                          (bind _returnvalue_3 (AddComponent|NODE_AddFortEmojiPreviewComponent-0 :self _returnvalue_2))))
                      (Emoji|PushPreviewEmoji _returnvalue_3 (Variables|Default|GetEmojiItemDefinition)))
                    (:Is%20Not%20Valid
                      _))
                  (Animation|ResetClothandDynamics _returnvalue_2)
                  (Variables|Default|SetStartSectionName (Emote|PickMontageSection _output_get _asanim_montage _returnvalue_2))
                  (Animation|PlayLocalAnimMontage _returnvalue_2 _asanim_montage 1.0 (Variables|Default|GetStartSectionName) true)
                  (Utilities|FlowControl|Sequence
                    (:then_0
                      (Utilities|FlowControl|Delay 0.1)
                      (CallFunction|ApplyEmotetoPet false))
                    (:then_1
                      _)))
                (:CastFailed
                  _))
              (Default|CallOnEmoteSetupFinished))))
        (:CastFailed
          (Variables|Default|SetEmojiItemDefinition 0))))))

(event EventEndPlay
  (params (EndPlayReason :type EEndPlayReason%20Enum :default "Destroyed"))
  (CallFunction|ApplyEmotetoPet true)
  (CallFunction|CleanupDynamicSounds))

(fn ApplyEmoteToPet
  (params (IsRemoval :type Boolean))
  (bind _returnvalue (FrontEndContext|GetHeroPlayerPawnforCurrentDisplayedItem (BlueprintContext|GetFortFrontEndContext)))
  (Utilities|IsValid _returnvalue
    (:Is%20Valid
      (Utilities|IsValid (Variables|/Script/FortniteGame.FortPlayerPawn|GetCosmeticPetInstance _returnvalue)
        (:Is%20Valid
          (bind _asfort_pet_anim_instance (Utilities|Casting|CastToFortPetAnimInstance (Components|SkeletalMesh|GetAnimInstance (Variables|/Script/FortniteGame.FortPlayerPet|GetPetMesh (Variables|/Script/FortniteGame.FortPlayerPawn|GetCosmeticPetInstance _returnvalue)))))
          (Class|FortPetAnimInstance|ApplyorRemoveFlagsfromEmote _asfort_pet_anim_instance (Variables|Default|GetDanceOrEmojiItemDefinition) (not IsRemoval)))))))

(fn CleanupDynamicSounds
  (bind _dynamicallyspawnedsounds (Variables|Default|GetDynamicallySpawnedSounds))
  (Audio|Components|Audio|Stop _dynamicallyspawnedsounds)
  (Components|DestroyComponent _dynamicallyspawnedsounds)
  (Utilities|Array|Clear (Variables|Default|GetDynamicallySpawnedSounds)))

(fn ConstructionScript)
